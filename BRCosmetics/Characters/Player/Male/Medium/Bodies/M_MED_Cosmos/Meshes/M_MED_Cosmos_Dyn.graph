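# Physics asset for the M_MED_Cosmos male character body cosmetic, covering the dynamic top holster bones.
# The dyn_holster_top body approximates its shape with two differently oriented capsules, while dyn_holster_top_MSTR carries only a tiny capsule.
# A single constraint joins the two bodies with both swing axes locked, leaving twist as the holster's rotational freedom relative to the master bone.
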PhysicsAsset /BRCosmetics/Characters/Player/Male/Medium/Bodies/M_MED_Cosmos/Meshes/M_MED_Cosmos_Dyn
bodies {
 dyn_holster_top [
  {shapeType=Capsule, length=24.4822, radius=2.5619, center=<13.8529 0 2.4504>, rotation=<rot 90 0 0>}
  {shapeType=Capsule, length=15.2188, radius=2.5938, center=<-0.9524 3.1405 2.7497>, rotation=<rot -90 0 -65.0045>}
 ]
 dyn_holster_top_MSTR [{shapeType=Capsule, length=0.505, radius=0.505, rotation=<rot 0 -90 0>}]
}
constraints [{
  bone1Name=dyn_holster_top
  bone2Name=dyn_holster_top_MSTR
  swing1LimitDegrees=45
  swing1Motion=Locked
  swing2LimitDegrees=45
  swing2Motion=Locked
  twistLimitDegrees=25
 }]
