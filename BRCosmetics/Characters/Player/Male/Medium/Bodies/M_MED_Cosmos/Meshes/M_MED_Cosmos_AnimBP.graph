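# Post-pose AnimBlueprint for the M_MED_Cosmos body, copying the attached parent mesh's pose and adding deformation and secondary motion.
# A twist Control Rig feeds an additive left shoulder pad offset, weighted by upperarm-versus-clavicle rotation computed each update.
# A weighted multi-bone constraint then drives the right-thigh holster, followed by RigidBody simulation relative to that thigh's twist bone.
# A sculpters Control Rig node sits unconnected.
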AnimBlueprint /BRCosmetics/Characters/Player/Male/Medium/Bodies/M_MED_Cosmos/Meshes/M_MED_Cosmos_AnimBP
target_skeleton=&/M_MED_Cosmos_Skeleton
graphs [
 AnimGraph {nodes [
   Control Rig() {
    alphaBoolBlend {blendOption=Linear}
    controlRigAssetReference {blueprintRigClass=&/M_MED_Cosmos_Deform_CtrlRig_sculpters.M_MED_Cosmos_Deform_CtrlRig_sculpters_C}
   }
   n1 = Control Rig(Source=Copy Pose From Mesh() {bUseAttachedParent=true}) {
    alphaBoolBlend {blendOption=Linear}
    controlRigAssetReference {blueprintRigClass=&/M_MED_Cosmos_Deform_CtrlRig_twists.M_MED_Cosmos_Deform_CtrlRig_twists_C}
   }
   n2 = "Transform (Modify) Bone - Bone: dyn_shoulderpad_l_MSTR"(
    ComponentPose=Local To Component(LocalPose=@n1)
    Alpha=Get arm rotation_l() -> (arm rotation_l:double)
    Translation="1,1,0"
    Rotation="0,0,10"
    Scale="1,1,1"
   ) {
    rotationMode=BMM_Additive
    rotationSpace=BCS_BoneSpace
    translationMode=BMM_Additive
    translationSpace=BCS_BoneSpace
    alphaBoolBlend {blendOption=Linear}
    alphaScaleBiasClamp {interpSpeedIncreasing=40}
    boneToModify {boneName=dyn_shoulderpad_l_MSTR}
    rotation {roll=10}
    translation=<1 1>
   }
   n3 = "Constraint - dyn_holster_bott_MSTR (PPPP)"!(
    ComponentPose=@n2
    Alpha=1
    ConstraintWeights_0=0.0946
    ConstraintWeights_1:float=0.0326
    ConstraintWeights_2:float=0.8522
    ConstraintWeights_3:float=0.0207
   ) -> (Pose) {
    alphaBoolBlend {blendOption=Linear}
    boneToModify {boneName=dyn_holster_bott_MSTR}
    constraintSetup [
     {targetBone {boneName=thigh_r}}
     {targetBone {boneName=pelvis}}
     {targetBone {boneName=thigh_twist_01_r}}
     {targetBone {boneName=deform_thigh_fr_r}}
    ]
    constraintWeights [1,1,1,1]
   }
   n4 = RigidBody!(ComponentPose=@n3, Alpha, OverrideWorldGravity:Vector="0,0,-980", ExternalForce) -> (Pose) {
    bClampLinearTranslationLimitToRefPose=true
    bOverrideWorldGravity=true
    overridePhysicsAsset=&/M_MED_Cosmos_Dyn
    simulationSpace=BaseBoneSpace
    alphaBoolBlend {blendOption=Linear}
    baseBoneRef {boneName=thigh_twist_01_r}
    componentAppliedLinearAccClamp=<3000 3000 3000>
    componentLinearAccScale {z=0.5}
    componentLinearVelScale {z=0.5}
    overrideWorldGravity {z=-980}
    simSpaceSettings {
     maxAngularAcceleration=1000
     maxAngularVelocity=3
     maxLinearAcceleration=250
     maxLinearVelocity=500
     velocityScaleZ=0
     worldAlpha=1
     externalLinearDragV=<0.5 0.5>
    }
   }
   Output Pose(Result=Component To Local(ComponentPose=@n4)) {layerGroup=None}
  ]}
 EventGraph {nodes [
   Event BlueprintUpdateAnimation()
   n1 = BreakTransform(InTransform=GetDeltaTransformFromRefPose(self=GetOwnerSkeletalMeshComponent(), BoneName=upperarm_l, BaseName=clavicle_l))
   > Set arm rotation_l(arm rotation_l:double=MapRangeClamped(Value=BreakRotator(InRot=@n1.Rotation), InRangeA=-18, InRangeB=-40)) -> (Output_Get:double)
  ]}
]
variables [arm rotation_l {type=number}]
