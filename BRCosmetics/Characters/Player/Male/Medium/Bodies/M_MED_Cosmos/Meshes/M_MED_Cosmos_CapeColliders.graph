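# Physics asset of collision shapes for the cape on the M_MED_Cosmos male player body, judging by its name.
# Capsules cover the pelvis, spine, clavicles and neck, with a sphere on the head and a holster-top capsule on dyn_blaster.
# Arm, leg and foot bodies have no shapes, and there are no constraints, so it serves as a collider set rather than a jointed ragdoll.
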PhysicsAsset /BRCosmetics/Characters/Player/Male/Medium/Bodies/M_MED_Cosmos/Meshes/M_MED_Cosmos_CapeColliders
bodies {
 calf_l []
 calf_r []
 clavicle_l [{shapeType=Capsule, length=9.0192, radius=9.8871, center=<9.354 0.2612 -2.5026>, rotation=<rot 69.3297 1.4476 176.9462>}]
 clavicle_r [{shapeType=Capsule, length=9.0192, radius=9.8871, center=<-9.354 -0.2612 2.5026>, rotation=<rot 69.3297 1.4476 -3.0538>}]
 dyn_blaster [{
   shapeType=Capsule
   length=20
   radius=9
   shapeName=dyn_holster_top_capsule
   center=<9.8721 -1.438 -0.2031>
   rotation=<rot 0.2946 72.7867 -92.974>
  }]
 foot_l []
 foot_r []
 head [{shapeType=Sphere, radius=8.4904, center=<11.028 -0.174 0>}]
 lowerarm_l []
 lowerarm_r []
 neck_02 [{shapeType=Capsule, length=13.4375, radius=7.3175, center=<-0.2642 -0.3408 0>, rotation=<rot 90 -171.1193 -175.3978>}]
 pelvis [{shapeType=Capsule, length=8.6456, radius=17.9724, center=<-2.4887 2.3673 0>}]
 spine_01 [{shapeType=Capsule, length=10.5058, radius=16.8776, center=<-1.7965 1.7568 0>}]
 spine_02 [
  {shapeType=Capsule, length=9.8884, radius=15.9303, center=<-1.711 -0.1851 0.0001>}
  {shapeType=Capsule, length=29.875, radius=11.1013, center=<2.4704 0 0>, rotation=<rot 90 -0.0396 -0.0396>}
 ]
 spine_03 [{shapeType=Capsule, length=11.7135, radius=16.3724, center=<0.1445 0.8772 0.0001>}]
 spine_04 [{shapeType=Capsule, length=10.2452, radius=16.3303, center=<2.766 1.7821 0.0002>}]
 spine_05 [{shapeType=Capsule, length=9.0058, radius=17.0729, center=<-5.4332 2.4149 0>}]
 thigh_l []
 thigh_r []
 upperarm_l []
 upperarm_r []
}
constraints []
