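# Physics asset for the dynamic cape on the M_MED_Cosmos male character cosmetic, with capsule bodies for the cape and the character's torso, arms and legs.
# Three seven-bone cape chains (center, left, right) hang from master bones constrained to backpack_root, widening toward the tips.
# Cross constraints tie each side bone to its center counterpart with tighter angular limits than along the chains, linking the strips like a sheet.
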PhysicsAsset /BRCosmetics/Characters/Player/Male/Medium/Bodies/M_MED_Cosmos/Meshes/Parts/M_MED_Cosmos_FaceAcc_Dyn_cape
bodies {
 backpack_root [{shapeType=Capsule, length=0.505, radius=0.505, rotation=<rot 0 -90 0>}]
 calf_l [{shapeType=Capsule, length=40, radius=19, center=<-30.8952 3.6326 0.0175>, rotation=<rot 90 180 178.4284>}]
 calf_r [{shapeType=Capsule, length=40, radius=19, center=<30.8952 -3.6326 -0.0175>, rotation=<rot 90 -179.8508 178.5776>}]
 clavicle_l [{shapeType=Capsule, length=8.0938, radius=8.63, center=<9.7772 -2.0948 -0.8642>, rotation=<rot 83.0193 10.3771 -0.0001>}]
 clavicle_r [{shapeType=Capsule, length=8.0938, radius=8.63, center=<-9.7047 2.108 0.8551>, rotation=<rot 83.0195 10.377 -0.0001>}]
 dyn_cape_01 [{shapeType=Capsule, length=10, radius=5, shapeName=dyn_cape_01_capsule, center=<7 0 3.0593>, rotation=<rot 90 0 0>}]
 dyn_cape_02 [{shapeType=Capsule, length=10, radius=7, shapeName=dyn_cape_01_capsule, center=<7 0 3.0593>, rotation=<rot 90 0 0>}]
 dyn_cape_03 [{shapeType=Capsule, length=10, radius=8.1308, shapeName=dyn_cape_01_capsule, center=<7 0 3.0593>, rotation=<rot 90 0 0>}]
 dyn_cape_04 [{shapeType=Capsule, length=10, radius=8.1308, shapeName=dyn_cape_01_capsule, center=<7 0 3.0593>, rotation=<rot 90 0 0>}]
 dyn_cape_05 [{shapeType=Capsule, length=10, radius=8.1308, shapeName=dyn_cape_01_capsule, center=<7 0 3.0593>, rotation=<rot 90 0 0>}]
 dyn_cape_06 [{shapeType=Capsule, length=10, radius=8.3, shapeName=dyn_cape_01_capsule, center=<7 0 3.0593>, rotation=<rot 90 0 0>}]
 dyn_cape_07 [{shapeType=Capsule, length=20, radius=8.5, shapeName=dyn_cape_01_capsule, center=<7 0 3.0593>, rotation=<rot 90 0 0>}]
 dyn_cape_MSTR [{shapeType=Capsule, length=0.505, radius=0.505, shapeName=dyn_cape_r_MSTR_capsule, rotation=<rot 0 -90 0>}]
 dyn_cape_l_01 [{shapeType=Capsule, length=10, radius=5, shapeName=dyn_cape_01_capsule, center=<7 0 3.0593>, rotation=<rot 90 0 0>}]
 dyn_cape_l_02 [{shapeType=Capsule, length=10, radius=7, shapeName=dyn_cape_01_capsule, center=<7 0 3.0593>, rotation=<rot 90 0 0>}]
 dyn_cape_l_03 [{shapeType=Capsule, length=10, radius=8.1308, shapeName=dyn_cape_01_capsule, center=<7 0 3.0593>, rotation=<rot 90 0 0>}]
 dyn_cape_l_04 [{shapeType=Capsule, length=10, radius=8.1308, shapeName=dyn_cape_01_capsule, center=<7 0 3.0593>, rotation=<rot 90 0 0>}]
 dyn_cape_l_05 [{shapeType=Capsule, length=10, radius=8.1308, shapeName=dyn_cape_01_capsule, center=<7 0 3.0593>, rotation=<rot 90 0 0>}]
 dyn_cape_l_06 [{shapeType=Capsule, length=10, radius=8.3, shapeName=dyn_cape_01_capsule, center=<7 0 3.0593>, rotation=<rot 90 0 0>}]
 dyn_cape_l_07 [{shapeType=Capsule, length=20, radius=8.5, shapeName=dyn_cape_01_capsule, center=<7 0 3.0593>, rotation=<rot 90 0 0>}]
 dyn_cape_l_MSTR [{shapeType=Capsule, length=0.505, radius=0.505, shapeName=dyn_cape_r_MSTR_capsule, rotation=<rot 0 -90 0>}]
 dyn_cape_r_01 [{shapeType=Capsule, length=10, radius=5, shapeName=dyn_cape_01_capsule, center=<7 0 3.0593>, rotation=<rot 90 0 0>}]
 dyn_cape_r_02 [{shapeType=Capsule, length=10, radius=7, shapeName=dyn_cape_01_capsule, center=<7 0 3.0593>, rotation=<rot 90 0 0>}]
 dyn_cape_r_03 [{shapeType=Capsule, length=10, radius=8.1308, shapeName=dyn_cape_01_capsule, center=<7 0 3.0593>, rotation=<rot 90 0 0>}]
 dyn_cape_r_04 [{shapeType=Capsule, length=10, radius=8.1308, shapeName=dyn_cape_01_capsule, center=<7 0 3.0593>, rotation=<rot 90 0 0>}]
 dyn_cape_r_05 [{shapeType=Capsule, length=10, radius=8.1308, shapeName=dyn_cape_01_capsule, center=<7 0 3.0593>, rotation=<rot 90 0 0>}]
 dyn_cape_r_06 [{shapeType=Capsule, length=10, radius=8.3, shapeName=dyn_cape_01_capsule, center=<7 0 3.0593>, rotation=<rot 90 0 0>}]
 dyn_cape_r_07 [{shapeType=Capsule, length=20, radius=8.5, shapeName=dyn_cape_01_capsule, center=<7 0 3.0593>, rotation=<rot 90 0 0>}]
 dyn_cape_r_MSTR [{shapeType=Capsule, length=0.505, radius=0.505, shapeName=dyn_cape_r_MSTR_capsule, rotation=<rot 0 -90 0>}]
 lowerarm_l [{shapeType=Capsule, length=31.9688, radius=8.63, center=<18.227 0.9178 -2.1609>, rotation=<rot 82.7121 -0.0002 3.2014>}]
 lowerarm_r [{shapeType=Capsule, length=31.4688, radius=8.9738, center=<-18.227 0.9402 2.1609>, rotation=<rot 82.7121 -0.0002 3.2014>}]
 pelvis [
  {shapeType=Capsule, length=11.1563, radius=19.1818, center=<-5.3453 4.1916 -0.0005>}
  {shapeType=Capsule, length=17.6562, radius=16.7238, center=<-21.8171 2.9919 0.0008>}
 ]
 spine_01 [{shapeType=Capsule, length=14.4688, radius=15.8693}]
 spine_02 [{shapeType=Capsule, length=15.375, radius=14.963, center=<0 1.3489 0>}]
 spine_03 [{shapeType=Capsule, length=16.0625, radius=14.838, center=<0 1.3604 0>}]
 spine_04 [{shapeType=Capsule, length=18.5, radius=12.6613, center=<5.8532 -0.5579 0.0001>}]
 spine_05 [{shapeType=Capsule, length=13.0312, radius=12.3488, center=<-1.8395 -1.386 -0.0003>}]
 thigh_l [{shapeType=Capsule, length=16.125, radius=18.5988, center=<-20.1925 4.0916 3.1216>, rotation=<rot 90 0 0>}]
 thigh_r [{shapeType=Capsule, length=16.125, radius=18.5988, center=<20.1925 -4.0916 -2.7025>, rotation=<rot 90 0 0>}]
 upperarm_l [{shapeType=Capsule, length=20.7842, radius=7.875, center=<10.6636 0 0>, rotation=<rot -83.203 -180 180>}]
 upperarm_r [{shapeType=Capsule, length=20.7842, radius=7.875, center=<-10.6636 0 0>, rotation=<rot -83.203 -180 180>}]
}
constraints [
 {
  bone1Name=clavicle_l
  bone2Name=spine_05
  swing1LimitDegrees=45
  swing1Motion=Free
  swing2LimitDegrees=45
  swing2Motion=Free
  twistLimitDegrees=45
  twistMotion=Free
 }
 {
  bone1Name=clavicle_r
  bone2Name=spine_05
  swing1LimitDegrees=45
  swing1Motion=Free
  swing2LimitDegrees=45
  swing2Motion=Free
  twistLimitDegrees=45
  twistMotion=Free
 }
 {
  bone1Name=backpack_root
  bone2Name=spine_05
  swing1LimitDegrees=45
  swing1Motion=Free
  swing2LimitDegrees=45
  swing2Motion=Free
  twistLimitDegrees=45
  twistMotion=Free
 }
 {
  bone1Name=spine_05
  bone2Name=spine_04
  swing1LimitDegrees=45
  swing1Motion=Free
  swing2LimitDegrees=45
  swing2Motion=Free
  twistLimitDegrees=45
  twistMotion=Free
 }
 {
  bone1Name=spine_04
  bone2Name=spine_03
  swing1LimitDegrees=45
  swing1Motion=Free
  swing2LimitDegrees=45
  swing2Motion=Free
  twistLimitDegrees=45
  twistMotion=Free
 }
 {
  bone1Name=spine_03
  bone2Name=spine_02
  swing1LimitDegrees=45
  swing1Motion=Free
  swing2LimitDegrees=45
  swing2Motion=Free
  twistLimitDegrees=45
  twistMotion=Free
 }
 {
  bone1Name=spine_02
  bone2Name=spine_01
  swing1LimitDegrees=45
  swing1Motion=Free
  swing2LimitDegrees=45
  swing2Motion=Free
  twistLimitDegrees=45
  twistMotion=Free
 }
 {
  bone1Name=spine_01
  bone2Name=pelvis
  swing1LimitDegrees=45
  swing1Motion=Free
  swing2LimitDegrees=45
  swing2Motion=Free
  twistLimitDegrees=45
  twistMotion=Free
 }
 {
  bone1Name=lowerarm_l
  bone2Name=upperarm_l
  swing1LimitDegrees=45
  swing1Motion=Free
  swing2LimitDegrees=45
  swing2Motion=Free
  twistLimitDegrees=45
  twistMotion=Free
 }
 {
  bone1Name=upperarm_l
  bone2Name=clavicle_l
  swing1LimitDegrees=45
  swing1Motion=Free
  swing2LimitDegrees=45
  swing2Motion=Free
  twistLimitDegrees=45
  twistMotion=Free
 }
 {
  bone1Name=lowerarm_r
  bone2Name=upperarm_r
  swing1LimitDegrees=45
  swing1Motion=Free
  swing2LimitDegrees=45
  swing2Motion=Free
  twistLimitDegrees=45
  twistMotion=Free
 }
 {
  bone1Name=upperarm_r
  bone2Name=clavicle_r
  swing1LimitDegrees=45
  swing1Motion=Free
  swing2LimitDegrees=45
  swing2Motion=Free
  twistLimitDegrees=45
  twistMotion=Free
 }
 {
  bone1Name=thigh_l
  bone2Name=pelvis
  swing1LimitDegrees=45
  swing1Motion=Free
  swing2LimitDegrees=45
  swing2Motion=Free
  twistLimitDegrees=45
  twistMotion=Free
 }
 {
  bone1Name=calf_l
  bone2Name=thigh_l
  swing1LimitDegrees=45
  swing1Motion=Free
  swing2LimitDegrees=45
  swing2Motion=Free
  twistLimitDegrees=45
  twistMotion=Free
 }
 {
  bone1Name=thigh_r
  bone2Name=pelvis
  swing1LimitDegrees=45
  swing1Motion=Free
  swing2LimitDegrees=45
  swing2Motion=Free
  twistLimitDegrees=45
  twistMotion=Free
 }
 {
  bone1Name=calf_r
  bone2Name=thigh_r
  swing1LimitDegrees=45
  swing1Motion=Free
  swing2LimitDegrees=45
  swing2Motion=Free
  twistLimitDegrees=45
  twistMotion=Free
 }
 {
  bone1Name=dyn_cape_MSTR
  bone2Name=backpack_root
  swing1LimitDegrees=45
  swing1Motion=Free
  swing2LimitDegrees=45
  swing2Motion=Free
  twistLimitDegrees=45
  twistMotion=Free
 }
 {bone1Name=dyn_cape_01, bone2Name=dyn_cape_MSTR, swing1LimitDegrees=45, swing2LimitDegrees=45, twistLimitDegrees=45}
 {bone1Name=dyn_cape_02, bone2Name=dyn_cape_01, swing1LimitDegrees=45, swing2LimitDegrees=45, twistLimitDegrees=45}
 {bone1Name=dyn_cape_03, bone2Name=dyn_cape_02, swing1LimitDegrees=45, swing2LimitDegrees=45, twistLimitDegrees=45}
 {bone1Name=dyn_cape_04, bone2Name=dyn_cape_03, swing1LimitDegrees=45, swing2LimitDegrees=45, twistLimitDegrees=45}
 {bone1Name=dyn_cape_05, bone2Name=dyn_cape_04, swing1LimitDegrees=45, swing2LimitDegrees=45, twistLimitDegrees=45}
 {bone1Name=dyn_cape_06, bone2Name=dyn_cape_05, swing1LimitDegrees=45, swing2LimitDegrees=45, twistLimitDegrees=45}
 {bone1Name=dyn_cape_07, bone2Name=dyn_cape_06, swing1LimitDegrees=45, swing2LimitDegrees=45, twistLimitDegrees=45}
 {
  bone1Name=dyn_cape_l_MSTR
  bone2Name=backpack_root
  swing1LimitDegrees=45
  swing1Motion=Free
  swing2LimitDegrees=45
  swing2Motion=Free
  twistLimitDegrees=45
  twistMotion=Free
 }
 {bone1Name=dyn_cape_l_01, bone2Name=dyn_cape_l_MSTR, swing1LimitDegrees=45, swing2LimitDegrees=45, twistLimitDegrees=45}
 {bone1Name=dyn_cape_l_02, bone2Name=dyn_cape_l_01, swing1LimitDegrees=45, swing2LimitDegrees=45, twistLimitDegrees=45}
 {bone1Name=dyn_cape_l_03, bone2Name=dyn_cape_l_02, swing1LimitDegrees=45, swing2LimitDegrees=45, twistLimitDegrees=45}
 {bone1Name=dyn_cape_l_04, bone2Name=dyn_cape_l_03, swing1LimitDegrees=45, swing2LimitDegrees=45, twistLimitDegrees=45}
 {bone1Name=dyn_cape_l_05, bone2Name=dyn_cape_l_04, swing1LimitDegrees=45, swing2LimitDegrees=45, twistLimitDegrees=45}
 {bone1Name=dyn_cape_l_06, bone2Name=dyn_cape_l_05, swing1LimitDegrees=45, swing2LimitDegrees=45, twistLimitDegrees=45}
 {bone1Name=dyn_cape_l_07, bone2Name=dyn_cape_l_06, swing1LimitDegrees=45, swing2LimitDegrees=45, twistLimitDegrees=45}
 {
  bone1Name=dyn_cape_r_MSTR
  bone2Name=backpack_root
  swing1LimitDegrees=45
  swing1Motion=Free
  swing2LimitDegrees=45
  swing2Motion=Free
  twistLimitDegrees=45
  twistMotion=Free
 }
 {bone1Name=dyn_cape_r_01, bone2Name=dyn_cape_r_MSTR, swing1LimitDegrees=45, swing2LimitDegrees=45, twistLimitDegrees=45}
 {bone1Name=dyn_cape_r_02, bone2Name=dyn_cape_r_01, swing1LimitDegrees=45, swing2LimitDegrees=45, twistLimitDegrees=45}
 {bone1Name=dyn_cape_r_03, bone2Name=dyn_cape_r_02, swing1LimitDegrees=45, swing2LimitDegrees=45, twistLimitDegrees=45}
 {bone1Name=dyn_cape_r_04, bone2Name=dyn_cape_r_03, swing1LimitDegrees=45, swing2LimitDegrees=45, twistLimitDegrees=45}
 {bone1Name=dyn_cape_r_05, bone2Name=dyn_cape_r_04, swing1LimitDegrees=45, swing2LimitDegrees=45, twistLimitDegrees=45}
 {bone1Name=dyn_cape_r_06, bone2Name=dyn_cape_r_05, swing1LimitDegrees=45, swing2LimitDegrees=45, twistLimitDegrees=45}
 {bone1Name=dyn_cape_r_07, bone2Name=dyn_cape_r_06, swing1LimitDegrees=45, swing2LimitDegrees=45, twistLimitDegrees=45}
 {bone1Name=dyn_cape_r_01, bone2Name=dyn_cape_01, swing1LimitDegrees=30, swing2LimitDegrees=30, twistLimitDegrees=30}
 {bone1Name=dyn_cape_l_01, bone2Name=dyn_cape_01, swing1LimitDegrees=30, swing2LimitDegrees=30, twistLimitDegrees=30}
 {bone1Name=dyn_cape_r_02, bone2Name=dyn_cape_02, swing1LimitDegrees=30, swing2LimitDegrees=30, twistLimitDegrees=30}
 {bone1Name=dyn_cape_l_02, bone2Name=dyn_cape_02, swing1LimitDegrees=30, swing2LimitDegrees=30, twistLimitDegrees=30}
 {bone1Name=dyn_cape_r_03, bone2Name=dyn_cape_03, swing1LimitDegrees=30, swing2LimitDegrees=30, twistLimitDegrees=30}
 {bone1Name=dyn_cape_l_03, bone2Name=dyn_cape_03, swing1LimitDegrees=30, swing2LimitDegrees=30, twistLimitDegrees=30}
 {bone1Name=dyn_cape_r_04, bone2Name=dyn_cape_04, swing1LimitDegrees=30, swing2LimitDegrees=30, twistLimitDegrees=30}
 {bone1Name=dyn_cape_r_05, bone2Name=dyn_cape_05, swing1LimitDegrees=30, swing2LimitDegrees=30, twistLimitDegrees=30}
 {bone1Name=dyn_cape_l_05, bone2Name=dyn_cape_05, swing1LimitDegrees=30, swing2LimitDegrees=30, twistLimitDegrees=30}
 {bone1Name=dyn_cape_r_06, bone2Name=dyn_cape_06, swing1LimitDegrees=30, swing2LimitDegrees=30, twistLimitDegrees=30}
 {bone1Name=dyn_cape_l_06, bone2Name=dyn_cape_06, swing1LimitDegrees=30, swing2LimitDegrees=30, twistLimitDegrees=30}
 {bone1Name=dyn_cape_r_07, bone2Name=dyn_cape_07, swing1LimitDegrees=30, swing2LimitDegrees=30, twistLimitDegrees=30}
 {bone1Name=dyn_cape_l_07, bone2Name=dyn_cape_07, swing1LimitDegrees=30, swing2LimitDegrees=30, twistLimitDegrees=30}
 {bone1Name=dyn_cape_l_04, bone2Name=dyn_cape_04, swing1LimitDegrees=30, swing2LimitDegrees=30, twistLimitDegrees=30}
]
