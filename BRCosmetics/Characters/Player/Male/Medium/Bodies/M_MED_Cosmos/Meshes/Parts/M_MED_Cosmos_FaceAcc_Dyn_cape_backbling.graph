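# Physics asset for the Cosmos male character's cape backbling, defining bodies and constraints for its dynamic cape bones.
# Four capsule chains hang from the cape clavicles and backpack root, with swing limits generally widening toward the hem.
# Free-angular cross constraints tie neighboring chains together along their lower segments, linking them like a sheet.
# Torso and limb capsules, a crouch sphere and a flat slope box provide collision shapes around the cape.
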PhysicsAsset /BRCosmetics/Characters/Player/Male/Medium/Bodies/M_MED_Cosmos/Meshes/Parts/M_MED_Cosmos_FaceAcc_Dyn_cape_backbling
bodies {
 backpack_root [{shapeType=Capsule, length=0.505, radius=0.505, rotation=<rot 0 -90 0>}]
 calf_l [{shapeType=Capsule, length=31.8125, radius=12.5363, center=<-31.4163 -2.6359 -0.0003>, rotation=<rot 90 0 -1.5716>}]
 calf_r [{shapeType=Capsule, length=31.8125, radius=12.5363, center=<31.4163 2.6359 0.0003>, rotation=<rot 90 0 -1.5716>}]
 cape_01_l [{shapeType=Capsule, length=0, radius=0.5}]
 cape_01_r [{shapeType=Capsule, length=0, radius=0.5}]
 cape_02_l [{shapeType=Capsule, length=15.0625, radius=10, center=<8.2848 0 10>, rotation=<rot -90 -0.0198 0.0198>}]
 cape_02_r [{shapeType=Capsule, length=15.0625, radius=10, center=<8.2848 0 10>, rotation=<rot -90 -0.0198 0.0198>}]
 cape_03_l [{shapeType=Capsule, length=21.0312, radius=15, center=<14.6169 -0.0002 15>, rotation=<rot -90 0.0099 -0.0099>}]
 cape_03_r [{shapeType=Capsule, length=21.0312, radius=15, center=<14.6169 -0.0002 15>, rotation=<rot -90 0.0099 -0.0099>}]
 cape_04_l [{shapeType=Capsule, length=13.0625, radius=15, center=<9.1357 0 15>, rotation=<rot -90 0.0132 -0.0132>}]
 cape_04_r [{shapeType=Capsule, length=13.0625, radius=15, center=<9.1357 0 15>, rotation=<rot -90 0.0132 -0.0132>}]
 cape_05_l [{shapeType=Capsule, length=10.75, radius=15, center=<8.333 0 15>, rotation=<rot -90 -0.0148 0.0148>}]
 cape_05_r [{shapeType=Capsule, length=10.75, radius=15, center=<8.333 0 15>, rotation=<rot -90 -0.0148 0.0148>}]
 cape_06_l [{shapeType=Capsule, length=10.75, radius=15, center=<5.7009 0 15>, rotation=<rot -90 -0.0099 0.0099>}]
 cape_06_r [{shapeType=Capsule, length=10.75, radius=15, center=<5.7009 0 15>, rotation=<rot -90 -0.0099 0.0099>}]
 cape_07_l [{shapeType=Capsule, length=10.75, radius=15, center=<6.0897 -0.0002 15>, rotation=<rot -90 0.0396 -0.0396>}]
 cape_07_r [{shapeType=Capsule, length=10.75, radius=15, center=<6.0897 -0.0002 15>, rotation=<rot -90 0.0396 -0.0396>}]
 cape_08_l [{shapeType=Capsule, length=10.75, radius=15, center=<4.992 0 15>, rotation=<rot -90 -0.0074 0.0074>}]
 cape_08_r [{shapeType=Capsule, length=10.75, radius=15, center=<4.992 0 15>, rotation=<rot -90 -0.0074 0.0074>}]
 cape_mid_01_l [{shapeType=Capsule, length=15.0625, radius=10, center=<9.7043 0 10>, rotation=<rot -90 -0.0198 0.0198>}]
 cape_mid_01_r [{shapeType=Capsule, length=15.0625, radius=10, center=<9.7043 0 10>, rotation=<rot -90 -0.0198 0.0198>}]
 cape_mid_02_l [{shapeType=Capsule, length=21.0312, radius=15, center=<14.0358 -0.0002 15>, rotation=<rot -90 -0.0297 0.0297>}]
 cape_mid_02_r [{shapeType=Capsule, length=21.0312, radius=15, center=<14.0358 -0.0002 15>, rotation=<rot -90 -0.0066 0.0066>}]
 cape_mid_03_l [{shapeType=Capsule, length=13.0625, radius=15, center=<8.6523 0 15>, rotation=<rot -90 0.0099 -0.0099>}]
 cape_mid_03_r [{shapeType=Capsule, length=13.0625, radius=15, center=<8.6523 0 15>, rotation=<rot -90 0.0099 -0.0099>}]
 cape_mid_04_l [{shapeType=Capsule, length=10.75, radius=15, center=<8.1677 0 15>, rotation=<rot -90 0.0049 -0.0049>}]
 cape_mid_04_r [{shapeType=Capsule, length=10.75, radius=15, center=<8.1677 0 15>, rotation=<rot -90 0.0049 -0.0049>}]
 cape_mid_05_l [{shapeType=Capsule, length=10.75, radius=15, center=<5.6974 -0.0001 15>, rotation=<rot -90 -0.0148 0.0148>}]
 cape_mid_05_r [{shapeType=Capsule, length=10.75, radius=15, center=<5.6974 -0.0001 15>, rotation=<rot -90 -0.0148 0.0148>}]
 cape_mid_06_l [{shapeType=Capsule, length=10.75, radius=15, center=<4.992 0 15>, rotation=<rot -90 -0.0074 0.0074>}]
 cape_mid_06_r [{shapeType=Capsule, length=10.75, radius=15, center=<4.992 0 15>, rotation=<rot -90 -0.0074 0.0074>}]
 cape_mid_07_l [{shapeType=Capsule, length=10.75, radius=15, center=<4.992 0 15>, rotation=<rot -90 -0.0074 0.0074>}]
 cape_mid_07_r [{shapeType=Capsule, length=10.75, radius=15, center=<4.992 0 15>, rotation=<rot -90 -0.0074 0.0074>}]
 clavicle_cape_l [{shapeType=Capsule, length=0.505, radius=0.505, rotation=<rot 0 -90 0>}]
 clavicle_cape_r [{shapeType=Capsule, length=0.505, radius=0.505, rotation=<rot 0 -90 0>}]
 clavicle_l [{shapeType=Capsule, length=0.505, radius=0.505, rotation=<rot 0 -90 0>}]
 clavicle_r [{shapeType=Capsule, length=0.505, radius=0.505, rotation=<rot 0 -90 0>}]
 dyn_crouch_collision [{shapeType=Sphere, radius=51.125, center=<-0.0006 6.727 0>}]
 dyn_root_slope [{shapeType=Box, extentX=231.0625, extentY=231.0625, extentZ=12.375, center=<0 0 -4.1865>}]
 lowerarm_l [{shapeType=Capsule, length=29.9062, radius=8.13, center=<18.227 0.9178 -2.1609>, rotation=<rot 82.7121 -0.0002 3.2014>}]
 lowerarm_r [
  {shapeType=Capsule, length=29.9062, radius=8.13, center=<-18.227 -0.9178 2.1609>, rotation=<rot 82.7121 -0.0002 3.2014>}
  {shapeType=Capsule, length=23.5938, radius=6.4062, center=<-31.9441 -1.2158 5.2412>, rotation=<rot 0 0 89.9994>}
 ]
 pelvis [
  {shapeType=Capsule, length=9.8125, radius=18.0363, center=<-5.4177 0 -0.0004>}
  {shapeType=Capsule, length=17.6562, radius=16.7238, center=<-21.9244 0 0.0008>}
 ]
 spine_01 [{shapeType=Capsule, length=20.4062, radius=16.2863, center=<0 2.708 0>}]
 spine_02 [{shapeType=Capsule, length=17.5625, radius=16.9113, center=<0 1.8395 0>}]
 spine_03 [{shapeType=Capsule, length=29.9062, radius=14.8488, center=<0.0678 -0.6209 0>}]
 spine_04 [{shapeType=Capsule, length=33.9375, radius=14.0363, center=<1.9705 -1.1595 0>}]
 spine_05 [{shapeType=Capsule, length=33.125, radius=11.755, center=<-1.1419 -1.6221 -0.0003>}]
 thigh_l [{shapeType=Capsule, length=16.125, radius=18.5988, center=<-20.2154 -0.6265 3.1209>, rotation=<rot 90 0 0>}]
 thigh_r [
  {shapeType=Capsule, length=16.125, radius=18.5988, center=<20.2154 0.6265 -2.7018>, rotation=<rot 90 0 0>}
  {shapeType=Capsule, length=17.4375, radius=12.5625, center=<4.6279 0.3683 8.354>, rotation=<rot -90 -2.3859 -17.6161>}
 ]
 upperarm_l [
  {shapeType=Capsule, length=21.4688, radius=6.38, center=<13.1767 -0.0002 0>, rotation=<rot 90 0 0>}
  {shapeType=Capsule, length=8.875, radius=7.875, center=<5.7273 -0.0002 0.1463>, rotation=<rot -83.203 -180 180>}
 ]
 upperarm_r [
  {shapeType=Capsule, length=21.4688, radius=6.38, center=<-13.1767 0.0002 0>, rotation=<rot 90 0 0>}
  {shapeType=Capsule, length=8.875, radius=7.875, center=<-5.7273 0.0002 -0.1463>, rotation=<rot -83.203 -180 180>}
 ]
}
constraints [
 {bone1Name=cape_02_l, bone2Name=cape_01_l, swing1LimitDegrees=5, swing2LimitDegrees=5, twistLimitDegrees=2.5}
 {bone1Name=cape_03_l, bone2Name=cape_02_l, swing1LimitDegrees=15, swing2LimitDegrees=10, twistLimitDegrees=5}
 {bone1Name=cape_04_l, bone2Name=cape_03_l, swing1LimitDegrees=30, swing2LimitDegrees=15, twistLimitDegrees=5}
 {bone1Name=cape_05_l, bone2Name=cape_04_l, swing1LimitDegrees=45, swing2LimitDegrees=45, twistLimitDegrees=5}
 {bone1Name=cape_06_l, bone2Name=cape_05_l, swing1LimitDegrees=45, swing2LimitDegrees=45, twistLimitDegrees=5}
 {bone1Name=cape_07_l, bone2Name=cape_06_l, swing1LimitDegrees=45, swing2LimitDegrees=45, twistLimitDegrees=5}
 {bone1Name=cape_08_l, bone2Name=cape_07_l, swing1LimitDegrees=45, swing2LimitDegrees=45, twistLimitDegrees=5}
 {bone1Name=cape_01_l, bone2Name=clavicle_cape_l, swing1LimitDegrees=45, swing2LimitDegrees=45, twistLimitDegrees=5}
 {bone1Name=cape_01_r, bone2Name=clavicle_cape_r, swing1LimitDegrees=45, swing2LimitDegrees=45, twistLimitDegrees=5}
 {bone1Name=cape_02_r, bone2Name=cape_01_r, swing1LimitDegrees=5, swing2LimitDegrees=5, twistLimitDegrees=2.5}
 {bone1Name=cape_03_r, bone2Name=cape_02_r, swing1LimitDegrees=15, swing2LimitDegrees=10, twistLimitDegrees=5}
 {bone1Name=cape_04_r, bone2Name=cape_03_r, swing1LimitDegrees=30, swing2LimitDegrees=15, twistLimitDegrees=5}
 {bone1Name=cape_05_r, bone2Name=cape_04_r, swing1LimitDegrees=45, swing2LimitDegrees=45, twistLimitDegrees=5}
 {bone1Name=cape_06_r, bone2Name=cape_05_r, swing1LimitDegrees=45, swing2LimitDegrees=45, twistLimitDegrees=5}
 {bone1Name=cape_07_r, bone2Name=cape_06_r, swing1LimitDegrees=45, swing2LimitDegrees=45, twistLimitDegrees=5}
 {bone1Name=cape_08_r, bone2Name=cape_07_r, swing1LimitDegrees=45, swing2LimitDegrees=45, twistLimitDegrees=5}
 {bone1Name=cape_mid_01_l, bone2Name=backpack_root, swing1LimitDegrees=2.5, swing2LimitDegrees=2.5, twistLimitDegrees=2.5}
 {bone1Name=cape_mid_02_l, bone2Name=cape_mid_01_l, swing1LimitDegrees=10, swing2LimitDegrees=5, twistLimitDegrees=5}
 {bone1Name=cape_mid_03_l, bone2Name=cape_mid_02_l, swing1LimitDegrees=15, swing2LimitDegrees=1, twistLimitDegrees=5}
 {bone1Name=cape_mid_04_l, bone2Name=cape_mid_03_l, swing1LimitDegrees=45, swing2LimitDegrees=10, twistLimitDegrees=5}
 {bone1Name=cape_mid_05_l, bone2Name=cape_mid_04_l, swing1LimitDegrees=45, swing2LimitDegrees=45, twistLimitDegrees=5}
 {bone1Name=cape_mid_06_l, bone2Name=cape_mid_05_l, swing1LimitDegrees=45, swing2LimitDegrees=45, twistLimitDegrees=5}
 {bone1Name=cape_mid_07_l, bone2Name=cape_mid_06_l, swing1LimitDegrees=45, swing2LimitDegrees=45, twistLimitDegrees=5}
 {bone1Name=cape_mid_01_r, bone2Name=backpack_root, swing1LimitDegrees=2.5, swing2LimitDegrees=2.5, twistLimitDegrees=2.5}
 {bone1Name=cape_mid_02_r, bone2Name=cape_mid_01_r, swing1LimitDegrees=10, swing2LimitDegrees=5, twistLimitDegrees=5}
 {bone1Name=cape_mid_03_r, bone2Name=cape_mid_02_r, swing1LimitDegrees=15, swing2LimitDegrees=1, twistLimitDegrees=5}
 {bone1Name=cape_mid_04_r, bone2Name=cape_mid_03_r, swing1LimitDegrees=45, swing2LimitDegrees=10, twistLimitDegrees=5}
 {bone1Name=cape_mid_05_r, bone2Name=cape_mid_04_r, swing1LimitDegrees=45, swing2LimitDegrees=45, twistLimitDegrees=5}
 {bone1Name=cape_mid_06_r, bone2Name=cape_mid_05_r, swing1LimitDegrees=45, swing2LimitDegrees=45, twistLimitDegrees=5}
 {bone1Name=cape_mid_07_r, bone2Name=cape_mid_06_r, swing1LimitDegrees=45, swing2LimitDegrees=45, twistLimitDegrees=5}
 {
  bone1Name=clavicle_cape_l
  bone2Name=clavicle_l
  swing1LimitDegrees=45
  swing1Motion=Free
  swing2LimitDegrees=45
  swing2Motion=Free
  twistLimitDegrees=45
  twistMotion=Free
 }
 {
  bone1Name=clavicle_cape_r
  bone2Name=clavicle_r
  swing1LimitDegrees=45
  swing1Motion=Free
  swing2LimitDegrees=45
  swing2Motion=Free
  twistLimitDegrees=45
  twistMotion=Free
 }
 {
  bone1Name=clavicle_l
  bone2Name=spine_05
  swing1LimitDegrees=45
  swing1Motion=Free
  swing2LimitDegrees=45
  swing2Motion=Free
  twistLimitDegrees=45
  twistMotion=Free
 }
 {
  bone1Name=clavicle_r
  bone2Name=spine_05
  swing1LimitDegrees=45
  swing1Motion=Free
  swing2LimitDegrees=45
  swing2Motion=Free
  twistLimitDegrees=45
  twistMotion=Free
 }
 {
  bone1Name=backpack_root
  bone2Name=spine_05
  swing1LimitDegrees=45
  swing1Motion=Free
  swing2LimitDegrees=45
  swing2Motion=Free
  twistLimitDegrees=45
  twistMotion=Free
 }
 {
  bone1Name=spine_05
  bone2Name=spine_04
  swing1LimitDegrees=45
  swing1Motion=Free
  swing2LimitDegrees=45
  swing2Motion=Free
  twistLimitDegrees=45
  twistMotion=Free
 }
 {
  bone1Name=spine_04
  bone2Name=spine_03
  swing1LimitDegrees=45
  swing1Motion=Free
  swing2LimitDegrees=45
  swing2Motion=Free
  twistLimitDegrees=45
  twistMotion=Free
 }
 {
  bone1Name=spine_03
  bone2Name=spine_02
  swing1LimitDegrees=45
  swing1Motion=Free
  swing2LimitDegrees=45
  swing2Motion=Free
  twistLimitDegrees=45
  twistMotion=Free
 }
 {
  bone1Name=spine_02
  bone2Name=spine_01
  swing1LimitDegrees=45
  swing1Motion=Free
  swing2LimitDegrees=45
  swing2Motion=Free
  twistLimitDegrees=45
  twistMotion=Free
 }
 {
  bone1Name=spine_01
  bone2Name=pelvis
  swing1LimitDegrees=45
  swing1Motion=Free
  swing2LimitDegrees=45
  swing2Motion=Free
  twistLimitDegrees=45
  twistMotion=Free
 }
 {
  bone1Name=lowerarm_l
  bone2Name=upperarm_l
  swing1LimitDegrees=45
  swing1Motion=Free
  swing2LimitDegrees=45
  swing2Motion=Free
  twistLimitDegrees=45
  twistMotion=Free
 }
 {
  bone1Name=upperarm_l
  bone2Name=clavicle_l
  swing1LimitDegrees=45
  swing1Motion=Free
  swing2LimitDegrees=45
  swing2Motion=Free
  twistLimitDegrees=45
  twistMotion=Free
 }
 {
  bone1Name=lowerarm_r
  bone2Name=upperarm_r
  swing1LimitDegrees=45
  swing1Motion=Free
  swing2LimitDegrees=45
  swing2Motion=Free
  twistLimitDegrees=45
  twistMotion=Free
 }
 {
  bone1Name=upperarm_r
  bone2Name=clavicle_r
  swing1LimitDegrees=45
  swing1Motion=Free
  swing2LimitDegrees=45
  swing2Motion=Free
  twistLimitDegrees=45
  twistMotion=Free
 }
 {
  bone1Name=thigh_l
  bone2Name=pelvis
  swing1LimitDegrees=45
  swing1Motion=Free
  swing2LimitDegrees=45
  swing2Motion=Free
  twistLimitDegrees=45
  twistMotion=Free
 }
 {
  bone1Name=calf_l
  bone2Name=thigh_l
  swing1LimitDegrees=45
  swing1Motion=Free
  swing2LimitDegrees=45
  swing2Motion=Free
  twistLimitDegrees=45
  twistMotion=Free
 }
 {
  bone1Name=thigh_r
  bone2Name=pelvis
  swing1LimitDegrees=45
  swing1Motion=Free
  swing2LimitDegrees=45
  swing2Motion=Free
  twistLimitDegrees=45
  twistMotion=Free
 }
 {
  bone1Name=calf_r
  bone2Name=thigh_r
  swing1LimitDegrees=45
  swing1Motion=Free
  swing2LimitDegrees=45
  swing2Motion=Free
  twistLimitDegrees=45
  twistMotion=Free
 }
 {
  bone1Name=cape_04_r
  bone2Name=cape_mid_03_r
  swing1LimitDegrees=45
  swing1Motion=Free
  swing2LimitDegrees=45
  swing2Motion=Free
  twistLimitDegrees=45
  twistMotion=Free
 }
 {
  bone1Name=cape_06_r
  bone2Name=cape_mid_05_r
  swing1LimitDegrees=45
  swing1Motion=Free
  swing2LimitDegrees=45
  swing2Motion=Free
  twistLimitDegrees=45
  twistMotion=Free
 }
 {
  bone1Name=cape_08_r
  bone2Name=cape_mid_07_r
  swing1LimitDegrees=45
  swing1Motion=Free
  swing2LimitDegrees=45
  swing2Motion=Free
  twistLimitDegrees=45
  twistMotion=Free
 }
 {
  bone1Name=cape_04_l
  bone2Name=cape_mid_03_l
  swing1LimitDegrees=45
  swing1Motion=Free
  swing2LimitDegrees=45
  swing2Motion=Free
  twistLimitDegrees=45
  twistMotion=Free
 }
 {
  bone1Name=cape_06_l
  bone2Name=cape_mid_05_l
  swing1LimitDegrees=45
  swing1Motion=Free
  swing2LimitDegrees=45
  swing2Motion=Free
  twistLimitDegrees=45
  twistMotion=Free
 }
 {
  bone1Name=cape_08_l
  bone2Name=cape_mid_07_l
  swing1LimitDegrees=45
  swing1Motion=Free
  swing2LimitDegrees=45
  swing2Motion=Free
  twistLimitDegrees=45
  twistMotion=Free
 }
 {
  bone1Name=cape_mid_03_l
  bone2Name=cape_mid_03_r
  swing1LimitDegrees=45
  swing1Motion=Free
  swing2LimitDegrees=45
  swing2Motion=Free
  twistLimitDegrees=45
  twistMotion=Free
 }
 {
  bone1Name=cape_mid_05_l
  bone2Name=cape_mid_05_r
  swing1LimitDegrees=45
  swing1Motion=Free
  swing2LimitDegrees=45
  swing2Motion=Free
  twistLimitDegrees=45
  twistMotion=Free
 }
 {
  bone1Name=cape_mid_07_l
  bone2Name=cape_mid_07_r
  swing1LimitDegrees=45
  swing1Motion=Free
  swing2LimitDegrees=45
  swing2Motion=Free
  twistLimitDegrees=45
  twistMotion=Free
 }
 {
  bone1Name=cape_05_r
  bone2Name=cape_mid_04_r
  swing1LimitDegrees=45
  swing1Motion=Free
  swing2LimitDegrees=45
  swing2Motion=Free
  twistLimitDegrees=45
  twistMotion=Free
 }
 {
  bone1Name=cape_07_r
  bone2Name=cape_mid_06_r
  swing1LimitDegrees=45
  swing1Motion=Free
  swing2LimitDegrees=45
  swing2Motion=Free
  twistLimitDegrees=45
  twistMotion=Free
 }
 {
  bone1Name=cape_05_l
  bone2Name=cape_mid_04_l
  swing1LimitDegrees=45
  swing1Motion=Free
  swing2LimitDegrees=45
  swing2Motion=Free
  twistLimitDegrees=45
  twistMotion=Free
 }
 {
  bone1Name=cape_07_l
  bone2Name=cape_mid_06_l
  swing1LimitDegrees=45
  swing1Motion=Free
  swing2LimitDegrees=45
  swing2Motion=Free
  twistLimitDegrees=45
  twistMotion=Free
 }
]
